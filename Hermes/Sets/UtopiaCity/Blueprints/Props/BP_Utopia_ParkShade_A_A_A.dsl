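; Park shade prop Blueprint for the Utopia City set, with an adjustable tilt on its shade component.
; Its construction script calls the parent, runs ValidateAdditionalComponents (an IsValid check on the shade component), then calls SetShadeAngle.
; SetShadeAngle rewrites the shade's relative transform with pitch taken from the ShadeAngle variable, preserving location, roll, yaw and scale.
; BeginPlay, overlap and Tick events are empty, so no runtime logic runs here.
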(event EventBeginPlay)

(event Collision|EventActorBeginOverlap
  (params (OtherActor :type Actor%20Object%20Reference)))

(event EventTick
  (params (DeltaSeconds :type Float%20%28single-precision%29)))

(fn SetShadeAngle
  (Transformation|SetRelativeTransform (Variables|Components|GetSM_Utopia_ParkShade_A) (Math|Transform|MakeTransform (.location (Transformation|GetRelativeTransform (Variables|Components|GetSM_Utopia_ParkShade_A))) (Math|Rotator|MakeRotator (.roll (.rotation (Transformation|GetRelativeTransform (Variables|Components|GetSM_Utopia_ParkShade_A)))) (Variables|Default|GetShadeAngle) (.yaw (.rotation (Transformation|GetRelativeTransform (Variables|Components|GetSM_Utopia_ParkShade_A))))) (.scale (Transformation|GetRelativeTransform (Variables|Components|GetSM_Utopia_ParkShade_A)))) false false))

(fn ConstructionScript
  (|CallParentFunction)
  (Default|ValidateAdditionalComponents)
  (Default|SetShadeAngle))

(fn ValidateAdditionalComponents
  (Utilities|IsValid (Variables|Components|GetSM_Utopia_ParkShade_A)))
