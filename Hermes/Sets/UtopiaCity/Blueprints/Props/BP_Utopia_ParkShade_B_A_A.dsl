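; Park shade prop Blueprint for the Utopia City set, with a shade mesh whose tilt is set by the ShadeAngle variable.
; The construction script calls its parent, runs a validity check on the shade mesh component, then calls SetShadeAngle.
; SetShadeAngle replaces the mesh's relative pitch with ShadeAngle, keeping location, roll, yaw and scale.
; BeginPlay, overlap and Tick events are empty.
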(event EventBeginPlay)

(event Collision|EventActorBeginOverlap
  (params (OtherActor :type Actor%20Object%20Reference)))

(event EventTick
  (params (DeltaSeconds :type Float%20%28single-precision%29)))

(fn SetShadeAngle
  (Transformation|SetRelativeTransform (Variables|Components|GetSM_Utopia_ParkShade_B) (Math|Transform|MakeTransform (.location (Transformation|GetRelativeTransform (Variables|Components|GetSM_Utopia_ParkShade_B))) (Math|Rotator|MakeRotator (.roll (.rotation (Transformation|GetRelativeTransform (Variables|Components|GetSM_Utopia_ParkShade_B)))) (Variables|Default|GetShadeAngle) (.yaw (.rotation (Transformation|GetRelativeTransform (Variables|Components|GetSM_Utopia_ParkShade_B))))) (.scale (Transformation|GetRelativeTransform (Variables|Components|GetSM_Utopia_ParkShade_B)))) false false))

(fn ConstructionScript
  (|CallParentFunction)
  (Default|ValidateAdditionalComponents)
  (Default|SetShadeAngle))

(fn ValidateAdditionalComponents
  (Utilities|IsValid (Variables|Components|GetSM_Utopia_ParkShade_B)))
